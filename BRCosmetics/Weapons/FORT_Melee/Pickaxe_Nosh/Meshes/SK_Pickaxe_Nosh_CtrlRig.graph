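# Control Rig for the Nosh two-handed pickaxe that procedurally animates its tongue, jaw and bottom-teeth bone chains.
# Three ChainHarmonics units run in sequence using only wave motion, with pendulum and reach disabled.
# A shared random float, remapped and smoothed by separate accumulating lerps, varies the Y wave amplitude of jaw and teeth, while the tongue's stays fixed.
# The slime bones in the hierarchy are not driven here.
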ControlRigBlueprint /BRCosmetics/Weapons/FORT_Melee/Pickaxe_Nosh/Meshes/SK_Pickaxe_Nosh_CtrlRig
graphs [
 RigVMFunctionLibrary {links [], nodes []}
 RigVMModel {nodes [
   n1 = &RigUnit_BeginExecution()
   /Script/ControlRig.RigUnit_ChainHarmonics(
    ChainRoot=dyn_tongue_base
    DrawWorldOffset=(Rotation=(X=0,Y=0,Z=0,W=1),Translation=(X=0,Y=0,Z=0),Scale3D=(X=1,Y=1,Z=1))
    Pendulum=(bEnabled=False,PendulumStiffness=2,PendulumGravity=(X=0,Y=0,Z=0),PendulumBlend=0.75,PendulumDrag=0.98,PendulumMinimum=0,PendulumMaximum=0.1,PendulumEase=Linear,UnwindAxis=(X=0,Y=1,Z=0),UnwindMinimum=0.2,UnwindMaximum=0.05)
    Reach=(bEnabled=False,ReachTarget=(X=0,Y=0,Z=0),ReachAxis=(X=1,Y=0,Z=0),ReachMinimum=0,ReachMaximum=0,ReachEase=SineEaseInOut)
    Speed=(X=0.2,Y=1.7,Z=0.2)
    Wave=(bEnabled=True,WaveFrequency=(X=1,Y=60,Z=1),WaveAmplitude=(X=0.25,Y=0.75,Z=0.25),WaveOffset=(X=1,Y=8,Z=1),WaveNoise=(X=0.5,Y=1,Z=0.5),WaveMinimum=1,WaveMaximum=0,WaveEase=SineEaseInOut)
    WaveCurve=(EditorCurveData=(Keys=((),(Time=1,Value=1)),PreInfinityExtrap=RCCE_Constant,PostInfinityExtrap=RCCE_Constant,DefaultValue=340282346638528859811704183484516925440))
    bDrawDebug=True
    ExecutePin=@n1.ExecutePin
   ) {is=struct RigVMUnitNode}
   n2 = /Script/RigVM.RigVMFunction_RandomFloat(Duration=4, Maximum=1, Minimum=0, Seed=217) {is=struct RigVMUnitNode, template="Random::Execute(in Seed,in Minimum,in Maximum,in Duration,out Result)"}
   > &RigUnit_ChainHarmonics(
    ChainRoot=dyn_jaw_02
    DrawWorldOffset=(Rotation=(X=0,Y=0,Z=0,W=1),Translation=(X=0,Y=0,Z=0),Scale3D=(X=1,Y=1,Z=1))
    Pendulum=(bEnabled=False,PendulumStiffness=2,PendulumGravity=(X=0,Y=0,Z=0),PendulumBlend=0.75,PendulumDrag=0.98,PendulumMinimum=0,PendulumMaximum=0.1,PendulumEase=Linear,UnwindAxis=(X=0,Y=1,Z=0),UnwindMinimum=0.2,UnwindMaximum=0.05)
    Reach=(bEnabled=False,ReachTarget=(X=0,Y=0,Z=0),ReachAxis=(X=1,Y=0,Z=0),ReachMinimum=0,ReachMaximum=0,ReachEase=SineEaseInOut)
    Speed=(X=0.25,Y=0.25,Z=0.25)
    Wave=(bEnabled=True,WaveFrequency=(X=0,Y=20,Z=0),WaveAmplitude=(X=0.2,Y=0.78,Z=0.2),WaveOffset=(X=0,Y=1,Z=0),WaveNoise=(X=0,Y=14,Z=0),WaveMinimum=1,WaveMaximum=0,WaveEase=SineEaseInOut)
    WaveCurve=(EditorCurveData=(Keys=((),(Time=1,Value=1)),PreInfinityExtrap=RCCE_Constant,PostInfinityExtrap=RCCE_Constant,DefaultValue=340282346638528859811704183484516925440))
    bDrawDebug=True
    Wave.WaveAmplitude.Y=/Script/RigVM.RigVMFunction_AccumulateFloatLerp(
     Blend=16
     InitialValue=1
     bIntegrateDeltaTime=true
     TargetValue=&RigVMFunction_MathFloatRemap(SourceMaximum=1, SourceMinimum=0, TargetMaximum=0.78, TargetMinimum=0, bClamp=False, Value=@n2.Result).Result
    ) {is=struct RigVMUnitNode, template="AccumulateLerp::Execute(in TargetValue,in InitialValue,in Blend,in bIntegrateDeltaTime,out Result)"}.Result
   )
   > &RigUnit_ChainHarmonics(
    ChainRoot=dyn_teeth_bottom_02
    DrawWorldOffset=(Rotation=(X=0,Y=0,Z=0,W=1),Translation=(X=0,Y=0,Z=0),Scale3D=(X=1,Y=1,Z=1))
    Pendulum=(bEnabled=False,PendulumStiffness=2,PendulumGravity=(X=0,Y=0,Z=0),PendulumBlend=0.75,PendulumDrag=0.98,PendulumMinimum=0,PendulumMaximum=0.1,PendulumEase=Linear,UnwindAxis=(X=0,Y=1,Z=0),UnwindMinimum=0.2,UnwindMaximum=0.05)
    Reach=(bEnabled=False,ReachTarget=(X=0,Y=0,Z=0),ReachAxis=(X=1,Y=0,Z=0),ReachMinimum=0,ReachMaximum=0,ReachEase=SineEaseInOut)
    Speed=(X=0,Y=0.3,Z=0)
    Wave=(bEnabled=True,WaveFrequency=(X=0,Y=20,Z=0),WaveAmplitude=(X=0,Y=0.25,Z=0),WaveOffset=(X=0,Y=0.75,Z=0),WaveNoise=(X=0,Y=8,Z=0),WaveMinimum=1,WaveMaximum=0,WaveEase=SineEaseInOut)
    WaveCurve=(EditorCurveData=(Keys=((),(Time=1,Value=1)),PreInfinityExtrap=RCCE_Constant,PostInfinityExtrap=RCCE_Constant,DefaultValue=340282346638528859811704183484516925440))
    bDrawDebug=True
    Wave.WaveAmplitude.Y=&RigVMFunction_AccumulateFloatLerp(
     Blend=7
     InitialValue=1
     bIntegrateDeltaTime=true
     TargetValue=&RigVMFunction_MathFloatRemap(SourceMaximum=1, SourceMinimum=0, TargetMaximum=0.35, TargetMinimum=0, bClamp=False, Value=@n2.Result).Result
    ) {template="AccumulateLerp::Execute(in TargetValue,in InitialValue,in Blend,in bIntegrateDeltaTime,out Result)"}.Result
   )
  ]}
]
hierarchy [
 b_AR_Root {loc=<0 0 0>, rot=<0 0 0.7071 0.7071>}
 Melee_TwoHanded_Handle {parent=b_AR_Root, loc=<8.1477 -0.008 -29.0793>, rot=<0 0 0.7071 -0.7071>}
 dyn_base {parent=Melee_TwoHanded_Handle, loc=<1.5295 -10.1796 95.754>, rot=<0.5 0.5 -0.5 0.5>}
 dyn_tongue_base {parent=dyn_base, loc=<-0.4328 0.0235 5.05>, rot=<0.6088 0 0.7934 0>}
 dyn_tongue_01 {parent=dyn_tongue_base, loc=<6.6412 1.0313 0.5429>, rot=<0 0.0436 0 -0.999>}
 dyn_tongue_02 {parent=dyn_tongue_01, loc=<5.3325 0.0598 0.5436>, rot=<0 0.0436 0 -0.999>}
 dyn_slime_base {parent=dyn_base, loc=<9.7844 0.0067 5.8807>, rot=<0 0 0 1>}
 dyn_slime_01 {parent=dyn_slime_base, loc=<-0.9019 9.4117 1.3074>, rot=<0 0 0 1>}
 dyn_slime_02 {parent=dyn_slime_base, loc=<-1.6942 -6.3273 14.5602>, rot=<0 0 0 1>}
 dyn_slime_03 {parent=dyn_slime_02, loc=<3.7409 -0.2627 0.1818>, rot=<0 0 0 1>}
 dyn_slime_04 {parent=dyn_slime_03, loc=<5.3768 -0.0211 -0.2827>, rot=<0 0 0 1>}
 dyn_slime_parent {parent=dyn_slime_base, loc=<-1.8341 -9.8435 -1.4765>, rot=<0 0 0.2588 -0.9659>}
 dyn_slime_05 {parent=dyn_slime_parent, loc=<0 0 0>, rot=<0 0 0 1>}
 dyn_slime_06 {parent=dyn_slime_05, loc=<5.9635 1.8965 -1.7738>, rot=<0 0 -0.2588 -0.9659>}
 dyn_slime_07 {parent=dyn_slime_06, loc=<6.6402 -0.7652 -0.5777>, rot=<0 0 0 1>}
 dyn_jaw_02 {parent=dyn_base, loc=<-1.2637 -1.5262 4.0283>, rot=<0 0.866 0 -0.5>}
 dyn_jaw_01 {parent=dyn_jaw_02, loc=<1.9923 0 0>, rot=<0 0 0 1>}
 dyn_jaw {parent=dyn_jaw_01, loc=<2.0505 0 0>, rot=<0 0 0 1>}
 dyn_teeth_bottom_02 {parent=dyn_base, loc=<4.0337 -0.3302 12.9126>, rot=<0 0 1 0>}
 dyn_teeth_bottom_01 {parent=dyn_teeth_bottom_02, loc=<0 0 2.5145>, rot=<0 0 0 1>}
 dyn_teeth_bottom {parent=dyn_teeth_bottom_01, loc=<0 0 2.81>, rot=<0 0 0 1>}
]
