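# Physics asset for the jaw mesh of the Nosh pickaxe, a melee cosmetic, covering its dyn_jaw bones.
# A wide, flat box body sits on dyn_jaw and a small cube body on dyn_jaw_01.
# One constraint joins the two with swing 1 and twist locked, leaving swing 2 as the only unlocked angular axis, so the jaw pivots like a hinge.
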PhysicsAsset /BRCosmetics/Weapons/FORT_Melee/Pickaxe_Nosh/Meshes/SK_Pickaxe_Nosh_Jaw_Physics
bodies {
 dyn_jaw [{shapeType=Box, extentX=10.1661, extentY=14.7183, extentZ=1.4804, center=<5.4618 0 -0.0001>}]
 dyn_jaw_01 [{shapeType=Box, extentX=2, extentY=2, extentZ=2}]
}
constraints [{
  bone1Name=dyn_jaw
  bone2Name=dyn_jaw_01
  swing1LimitDegrees=45
  swing1Motion=Locked
  swing2LimitDegrees=30
  twistLimitDegrees=45
  twistMotion=Locked
 }]
