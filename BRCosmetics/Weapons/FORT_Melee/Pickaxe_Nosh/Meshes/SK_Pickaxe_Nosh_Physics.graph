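# Physics asset for the Nosh pickaxe skeletal mesh, giving collision bodies and joint limits to its dyn_ bones for the slime and bottom teeth.
# The base body combines two capsules and a box, the bottom teeth use a box, and each slime bone has a capsule.
# Angle-limited constraints hang the tightly restricted teeth, one single slime bone and two three-bone slime chains off the base.
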PhysicsAsset /BRCosmetics/Weapons/FORT_Melee/Pickaxe_Nosh/Meshes/SK_Pickaxe_Nosh_Physics
bodies {
 dyn_base [
  {shapeType=Capsule, length=11.67, radius=3.9623, shapeName=dyn_base_capsule, center=<20.9061 -3.8711 -10.4532>, rotation=<rot -8.9293 0 0>}
  {
   shapeType=Capsule
   length=7.8145
   radius=3.8048
   shapeName=dyn_base_capsule
   center=<22.0065 -0.6019 -4.698>
   rotation=<rot -8.9293 0 89.2075>
  }
  {shapeType=Box, extentX=1.5, extentY=1.5, extentZ=1.5}
 ]
 dyn_slime_01 [{shapeType=Capsule, length=2.4007, radius=2.0369, center=<2.739 0 0>, rotation=<rot 90 0 0>}]
 dyn_slime_02 [{shapeType=Capsule, length=1.9862, radius=1.7087, center=<2.3032 0 0.0001>, rotation=<rot 90 0 0>}]
 dyn_slime_03 [{shapeType=Capsule, length=1.9862, radius=1.7087, center=<2.6459 0 0.0001>, rotation=<rot 90 0 0>}]
 dyn_slime_04 [{shapeType=Capsule, length=1.9862, radius=1.7087, center=<2.6459 0 0.0001>, rotation=<rot 90 0 0>}]
 dyn_slime_05 [{shapeType=Capsule, length=2.9892, radius=2.3306, center=<3.9639 0 0>, rotation=<rot 90 0 0>}]
 dyn_slime_06 [{shapeType=Capsule, length=2.9892, radius=2.3306, center=<3.5195 0 0>, rotation=<rot 90 0 0>}]
 dyn_slime_07 [{shapeType=Capsule, length=2.9892, radius=2.3306, center=<3.5195 0 0>, rotation=<rot 90 0 0>}]
 dyn_teeth_bottom [{shapeType=Box, extentX=4.6382, extentY=6.5387, extentZ=1.821, center=<2.5802 0 0>}]
}
constraints [
 {bone1Name=dyn_teeth_bottom, bone2Name=dyn_base, swing1LimitDegrees=5, swing2LimitDegrees=10, twistLimitDegrees=2}
 {bone1Name=dyn_slime_01, bone2Name=dyn_base, swing1LimitDegrees=20, swing2LimitDegrees=20, twistLimitDegrees=8}
 {bone1Name=dyn_slime_02, bone2Name=dyn_base, swing1LimitDegrees=30, swing2LimitDegrees=30, twistLimitDegrees=20}
 {bone1Name=dyn_slime_03, bone2Name=dyn_slime_02, swing1LimitDegrees=40, swing2LimitDegrees=40, twistLimitDegrees=20}
 {bone1Name=dyn_slime_06, bone2Name=dyn_slime_05, swing1LimitDegrees=20, swing2LimitDegrees=20, twistLimitDegrees=10}
 {bone1Name=dyn_slime_07, bone2Name=dyn_slime_06, swing1LimitDegrees=20, swing2LimitDegrees=20, twistLimitDegrees=10}
 {bone1Name=dyn_slime_04, bone2Name=dyn_slime_03, swing1LimitDegrees=35, swing2LimitDegrees=35, twistLimitDegrees=20}
 {bone1Name=dyn_slime_05, bone2Name=dyn_base, swing1LimitDegrees=35, swing2LimitDegrees=35, twistLimitDegrees=12}
]
