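# Physics asset for the tongue of the Nosh pickaxe skeletal mesh, defining the bodies and joints for its physical motion.
# A capsule on dyn_base anchors a chain of three identical boxes running from dyn_tongue_base through dyn_tongue_01 to dyn_tongue_02.
# Each joint locks swing1 and sets small twist and swing2 limit angles, with the swing2 angle widening toward the tip.
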PhysicsAsset /BRCosmetics/Weapons/FORT_Melee/Pickaxe_Nosh/Meshes/SK_Pickaxe_Nosh_Tongue_Physics
bodies {
 dyn_base [{shapeType=Capsule, length=21.625, radius=8.375, center=<9.6878 0 11.8837>, rotation=<rot 0 0 90>}]
 dyn_tongue_01 [{shapeType=Box, extentX=4.8699, extentY=6.502, extentZ=1.5905, center=<2.8415 0 -0.0002>}]
 dyn_tongue_02 [{shapeType=Box, extentX=4.8699, extentY=6.502, extentZ=1.5905, center=<2.8415 0 -0.0002>}]
 dyn_tongue_base [{shapeType=Box, extentX=4.8699, extentY=6.502, extentZ=1.5905, center=<2.8415 0 -0.0002>}]
}
constraints [
 {bone1Name=dyn_tongue_base, bone2Name=dyn_base, swing1LimitDegrees=45, swing1Motion=Locked, swing2LimitDegrees=2, twistLimitDegrees=2}
 {bone1Name=dyn_tongue_01, bone2Name=dyn_tongue_base, swing1LimitDegrees=45, swing1Motion=Locked, swing2LimitDegrees=4, twistLimitDegrees=2}
 {bone1Name=dyn_tongue_02, bone2Name=dyn_tongue_01, swing1LimitDegrees=45, swing1Motion=Locked, swing2LimitDegrees=6, twistLimitDegrees=2}
]
