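# Animation Blueprint for the Nosh pickaxe that adds secondary physics motion to its jaw and tongue over the cached input pose.
# When moving, it chains jaw, half-weight tongue and main RigidBody simulations plus a tongue rotation, with constant upward force on the jaw simulation.
# When still, it uses the main simulation, jaw and tongue offsets and a Control Rig.
# Player 0's velocity, not the owner's, selects the branch.
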AnimBlueprint /BRCosmetics/Weapons/FORT_Melee/Pickaxe_Nosh/Meshes/SK_Pickaxe_Nosh_AnimBp
target_skeleton=&/SK_Pickaxe_Nosh_Skeleton
graphs [
 AnimGraph {nodes [
   n1 = RigidBody(ComponentPose=Local To Component(LocalPose=Use cached pose 'NoshPickaxe_Cache'()), ExternalForce=Get Ext_Force() -> (Ext_Force:Vector)) {
    bClampLinearTranslationLimitToRefPose=true
    bOverrideWorldGravity=true
    overridePhysicsAsset=&/SK_Pickaxe_Nosh_Jaw_Physics
    simulationSpace=BaseBoneSpace
    alphaBoolBlend {blendOption=Linear}
    baseBoneRef {boneName=b_AR_Root}
    componentAppliedLinearAccClamp=<2000 2000 2000>
    componentLinearAccScale=<0.25 0.25 0.5>
    componentLinearVelScale=<0.25 0.25 0.5>
    overrideWorldGravity {z=-800}
   }
   n2 = RigidBody(ComponentPose=@n1, Alpha=0.5) {
    bClampLinearTranslationLimitToRefPose=true
    bOverrideWorldGravity=true
    overridePhysicsAsset=&/SK_Pickaxe_Nosh_Tongue_Physics
    simulationSpace=BaseBoneSpace
    alphaBoolBlend {blendOption=Linear}
    baseBoneRef {boneName=b_AR_Root}
    componentAppliedLinearAccClamp=<4000 4000 4000>
    componentLinearAccScale=<0.25 0.25 1>
    componentLinearVelScale=<0.25 0.25 1>
    overrideWorldGravity {z=-725}
   }
   n3 = RigidBody(ComponentPose=@n2) {
    bClampLinearTranslationLimitToRefPose=true
    bOverrideWorldGravity=true
    overridePhysicsAsset=&/SK_Pickaxe_Nosh_Physics
    simulationSpace=BaseBoneSpace
    alphaBoolBlend {blendOption=Linear}
    baseBoneRef {boneName=b_AR_Root}
    componentAppliedLinearAccClamp=<6000 6000 6000>
    componentLinearAccScale=<0.25 0.25 0.25>
    componentLinearVelScale=<0.25 0.25 0.25>
    overrideWorldGravity {z=-980}
   }
   n4 = "Transform (Modify) Bone - Bone: dyn_tongue_base"(ComponentPose=@n3, Alpha=1, Translation="0,0,0", Rotation="8,0,0", Scale="1,1,1") {
    rotationMode=BMM_Additive
    rotationSpace=BCS_BoneSpace
    alphaBoolBlend {blendOption=Linear}
    boneToModify {boneName=dyn_tongue_base}
    rotation {pitch=8}
   }
   n5 = RigidBody(ComponentPose=Local To Component(LocalPose=Use cached pose 'NoshPickaxe_Cache'())) {
    bClampLinearTranslationLimitToRefPose=true
    bOverrideWorldGravity=true
    overridePhysicsAsset=&SK_Pickaxe_Nosh_Physics
    simulationSpace=BaseBoneSpace
    alphaBoolBlend {blendOption=Linear}
    baseBoneRef {boneName=b_AR_Root}
    componentAppliedLinearAccClamp=<6000 6000 6000>
    componentLinearAccScale=<0.25 0.25 0.25>
    componentLinearVelScale=<0.25 0.25 0.25>
    overrideWorldGravity {z=-980}
   }
   n6 = "Transform (Modify) Bone - Bone: dyn_jaw"(ComponentPose=@n5, Alpha=1, Translation="0,0,0", Rotation="-5,0,0", Scale="1,1,1") {
    rotationMode=BMM_Additive
    rotationSpace=BCS_BoneSpace
    alphaBoolBlend {blendOption=Linear}
    boneToModify {boneName=dyn_jaw}
    rotation {pitch=-5}
   }
   n7 = "Transform (Modify) Bone - Bone: dyn_tongue_base"(ComponentPose=@n6, Alpha=1, Translation="0,0,0", Rotation="10,0,0", Scale="1,1,1") {
    rotationMode=BMM_Additive
    rotationSpace=BCS_BoneSpace
    alphaBoolBlend {blendOption=Linear}
    boneToModify {boneName=dyn_tongue_base}
    rotation {pitch=10}
   }
   n8 = Control Rig(Source=Component To Local(ComponentPose=@n7)) {
    alphaBoolBlend {blendOption=Linear}
    controlRigAssetReference {blueprintRigClass=&/SK_Pickaxe_Nosh_CtrlRig.SK_Pickaxe_Nosh_CtrlRig_C}
   }
   n9 = Get IsMovingZ() -> (IsMovingZ:bool)
   n10 = Get IsMovingX() -> (IsMovingX:bool)
   n11 = Get IsMovingY() -> (IsMovingY:bool)
   Output Pose(Result=Blend Poses by bool(BlendPose_0=Component To Local(ComponentPose=@n4), BlendPose_1=@n8, bActiveValue=OR Boolean!(A=@n10, B=@n11, C:bool=@n9) -> (ReturnValue)) {blendType=Linear}) {layerGroup=None}
   Save cached pose 'NoshPickaxe_Cache'(Pose=Input Pose()) {cacheName=NoshPickaxe_Cache}
  ]}
 EventGraph {nodes [
   Event BlueprintUpdateAnimation()
   > n1 = Sequence()
   > Set Ext_Force(Ext_Force:Vector="0, 0,1250") -> (Output_Get:Vector)
   n2 = Set velocity(
    execute=@n1.then_1
    velocity:Vector=GetVelocity!(self:Actor Object Reference=GetPlayerCharacter!(PlayerIndex:int=0) -> (ReturnValue:Character Object Reference)) -> (ReturnValue:Vector)
   ) -> (Output_Get:Vector)
   n3 = BreakVector(InVec=@n2)
   > Set IsMovingX(IsMovingX:bool="Not Equal (Float)"!(A:double=@n3, B:double=0) -> (ReturnValue:bool) {is=title K2Node_CallFunction}) -> (Output_Get:bool)
   > Set IsMovingY(IsMovingY:bool="Not Equal (Float)"(A=@n3.Y, B=0)) -> (Output_Get:bool)
   > Set IsMovingZ(IsMovingZ:bool="Not Equal (Float)"(A=@n3.Z, B=0)) -> (Output_Get:bool)
  ]}
]
variables [
 Ext_Force {type=Vector}
 IsMovingX {type=boolean}
 IsMovingY {type=boolean}
 IsMovingZ {type=boolean}
 velocity {type=Vector}
]
